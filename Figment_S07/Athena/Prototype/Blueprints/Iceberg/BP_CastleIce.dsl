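; BP_CastleIce is a prototype castle iceberg actor whose mesh transform reflects live calendar event progress.
; At match start, unless Debug is set, the server watches two events and latches Time once: Event1's elapsed fraction, 1 if Event2 is active, or 0 after a 3-second fallback.
; OnRep_Time maps Time through a curve to interpolate between two transforms, so the ice does not animate continuously.
; The construction script applies the same interpolation for preview.
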(event EventBeginPlay
  (Collision|SetCollisionResponsetoChannel (Variables|Default|GetStaticMesh) "ECC_GameTraceChannel13" "ECR_Block")
  (Networking|SwitchHasAuthority
    (:Authority
      (bind _asfort_game_mode_zone (Utilities|Casting|CastToFortGameModeZone (Game|GetGameMode)))
      (EventDispatchers|BindEventtoOnHandleMatchHasStarted _asfort_game_mode_zone (delegate Custom|BindToCalendarEvents)))))

(event Custom|BindToCalendarEvents
  (if (not (Variables|Default|GetDebug))
    (Utilities|FlowControl|Sequence
      (:then_0
        (bind (_eventname _timeuntilend _timesincebegin) (MCPContext|WatchCalendarEvent (Variables|Default|GetEvent1))
          (:then
            (Utilities|FlowControl|Delay 3.0)
            (if (not (Variables|Default|GetValueSet))
              (Variables|Default|SetTime 0.0)
              (Variables|Default|SetValueSet true)))
          (:OnEventActive
            (if (not (Variables|Default|GetValueSet))
              (Variables|Default|SetTime (Math|Timespan|TimespanRatio _timesincebegin (Math|Timespan|Timespan+Timespan _timeuntilend _timesincebegin)))
              (Variables|Default|SetValueSet true)))))
      (:then_1
        (MCPContext|WatchCalendarEvent (Variables|Default|GetEvent2)
          (:OnEventActive
            (if (not (Variables|Default|GetValueSet))
              (Variables|Default|SetTime 1.0)
              (Variables|Default|SetValueSet true))))))))

(fn OnRep_Time
  (Transformation|SetRelativeTransform (Variables|Default|GetStaticMesh) (Math|Transform|Lerp(Transform) (Variables|Default|GetTransform) (Variables|Default|GetTransform2) (Math|Curves|GetFloatValue (Variables|Default|GetCurve) (Variables|Default|GetTime))) false false))

(fn ConstructionScript
  (Components|StaticMesh|SetStaticMesh (Variables|Default|GetStaticMesh) (Variables|Default|GetScaledStaticmesh))
  (if (== (Variables|Default|GetTime) -1.0)
    (Transformation|SetRelativeTransform (Variables|Default|GetStaticMesh) (Variables|Default|GetTransform) false false)
    (else
      (Transformation|SetRelativeTransform (Variables|Default|GetStaticMesh) (Math|Transform|Lerp(Transform) (Variables|Default|GetTransform) (Variables|Default|GetTransform2) (Math|Curves|GetFloatValue (Variables|Default|GetCurve) (Variables|Default|GetTime))) false false))))
