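; Iceberg actor for Polar Peaks that shows staged ice melting driven by calendar events.
; After match start, the server watches C0–C4 and C3A/C4A, setting phase variables whose OnReps swap meshes or hide the ice.
; Late phases lerp the mesh transform along a curve by event progress, with C3A/C4A forcing the end state.
; IceMeltPhase OnRep updates a UI helper's ice stamp, and the construction script previews debug phases.
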(event Custom|IceMelt_C0
  ;[0 0 0 0] IceMeltA(C0)
  (Variables|Default|SetPhase0_Active? true)
  (Variables|Default|SetIceMeltPhase 0))

(event Custom|IceMelt_C1
  ;[0 0 0 0] IceMeltB(C1)
  (Variables|Default|SetPhase1_Active? true)
  (Variables|Default|SetIceMeltPhase 1))

(event Custom|IceMelt_C2
  ;[0 0 0 0] IceMelltC(C2)
  (Variables|Default|SetPhase2_Active? true))

(event EventBeginPlay
  ;[1 1 1] Watch for Calendar Events / Server
  (Collision|SetCollisionResponsetoChannel (Variables|Default|GetStaticMesh) "ECC_GameTraceChannel13" "ECR_Block")
  (Networking|SwitchHasAuthority
    (:Authority
      (bind _asfort_game_mode_zone (Utilities|Casting|CastToFortGameModeZone (Game|GetGameMode)))
      (EventDispatchers|BindEventtoOnHandleMatchHasStarted _asfort_game_mode_zone (delegate Custom|BindToCalendarEvents)))))

(event Custom|BindToCalendarEvents
  (Utilities|FlowControl|Sequence
    (:then_0
      ;[0 0 0 0] IceMeltA(C0)
      (MCPContext|WatchCalendarEvent "C0"
        (:OnEventActive
          (Variables|Default|SetPhase0_Active? true)
          (Variables|Default|SetIceMeltPhase 0))) ; Beacon Event
    )
    (:then_1
      ;[0 0 0 0] IceMeltB(C1)
      (MCPContext|WatchCalendarEvent "C1"
        (:OnEventActive
          (Variables|Default|SetPhase1_Active? true)
          (Variables|Default|SetIceMeltPhase 1))) ; Beacon Event
    )
    (:then_2
      ;[0 0 0 0] IceMelltC(C2)
      (MCPContext|WatchCalendarEvent "C2"
        (:OnEventActive
          (Variables|Default|SetPhase2_Active? true))) ; Beacon Event
    )
    (:then_3
      ;[0 0 0 0.5] IceMeltD (C3)
      (bind (_eventname _timeuntilend _timesincebegin) (MCPContext|WatchCalendarEvent "C3")
        (:OnEventActive
          (if (not (Variables|Default|GetValueSet))
            (Variables|Default|SetTime_D (Math|Timespan|TimespanRatio _timesincebegin (Math|Timespan|Timespan+Timespan _timeuntilend _timesincebegin)))
            (Variables|Default|SetPhase3_Active? true)))) ; Beacon Event
    )
    (:then_4
      ;[0 0 0 0.5] IceMeltE (C4)
      (bind (_eventname_2 _timeuntilend_2 _timesincebegin_2) (MCPContext|WatchCalendarEvent "C4")
        (:OnEventActive
          (if (not (Variables|Default|GetValueSet))
            (Variables|Default|SetTime_E (Math|Timespan|TimespanRatio _timesincebegin_2 (Math|Timespan|Timespan+Timespan _timeuntilend_2 _timesincebegin_2)))
            (Variables|Default|SetPhase4_Active? true)))) ; Beacon Event
    )
    (:then_5
      ;[0 0 0 0.5] C3A & C4A Event
      (bind _eventname_3 (MCPContext|WatchCalendarEvents (Utilities|Array|MakeArray "C3A" "C4A"))
        (:OnEventActive
          (if (Variables|Default|GetValueSet)
            _
            (elif (Utilities|String|EqualExactly(String) _eventname_3 "C3A")
              (Variables|Default|SetTime_D 1.0)
              (Variables|Default|SetPhase3_Active? true)
              (else
                (Variables|Default|SetTime_E 1.0)
                (Variables|Default|SetPhase4_Active? true)))))))))

(event Custom|IceMelt_C3
  ;[0 0 0 0.5] IceMeltD (C3)
  (Variables|Default|SetTime_D (Math|Timespan|TimespanRatio (stale Custom|BindToCalendarEvents _timesincebegin) (Math|Timespan|Timespan+Timespan (stale Custom|BindToCalendarEvents _timeuntilend) (stale Custom|BindToCalendarEvents _timesincebegin))))
  (Variables|Default|SetPhase3_Active? true))

(event Custom|IceMelt_C4
  ;[0 0 0 0.5] IceMeltE (C4)
  (Variables|Default|SetTime_E (Math|Timespan|TimespanRatio (stale Custom|BindToCalendarEvents _timesincebegin_2) (Math|Timespan|Timespan+Timespan (stale Custom|BindToCalendarEvents _timeuntilend_2) (stale Custom|BindToCalendarEvents _timesincebegin_2))))
  (Variables|Default|SetPhase4_Active? true))

(event Custom|IceMelt_C3A
  ;[0 0 0 0.5] C3A & C4A Event
  (Variables|Default|SetTime_D 1.0)
  (Variables|Default|SetPhase3_Active? true))

(event Custom|IceMelt_C4A
  ;[0 0 0 0.5] C3A & C4A Event
  (Variables|Default|SetTime_E 1.0)
  (Variables|Default|SetPhase4_Active? true))

(event Custom|NoIce
  ;[0 0 0 0.5] If nothing is Active
  (Utilities|FlowControl|Delay 3.0)
  (if (not (Variables|Default|GetValueSet))
    (Variables|Default|SetPhase5_Active? true)))

(fn IsPhase?
  (params (Phase :type Integer))
  (if (== Phase (Variables|Default|GetDebug_Phase))
    (return)))

(fn OnRep_IceMeltPhase
  (Networking|CanExecuteCosmeticEvents
    (:True
      (bind _polarpeaksuihelper (Variables|Default|GetPolarPeaksUIHelper)
        (:then
          _)
        (:else
          (bind _returnvalue (Game|SpawnActorfromClass "/Figment_S07/Athena/Prototype/Blueprints/Iceberg/BP_PolarPeaks_UIHelper_Actor.BP_PolarPeaks_UIHelper_Actor_C" (struct :Location "0, 0, 0" :Rotation "0, 0, 0" :Scale "1.000000,1.000000,1.000000") "AlwaysSpawn" "OverrideRootScale" 0 0))
          (bind _output_get (Variables|Default|SetPolarPeaksUIHelper _returnvalue))))
      (Class|BPPolarPeaksUIHelperActor|SetupPolarPeaksIceStamp _polarpeaksuihelper (Transformation|GetActorLocation) (Variables|Default|GetIceMeltPhase)))))

(fn OnRep_Phase0_Active?
  (Components|StaticMesh|SetStaticMesh (Variables|Default|GetStaticMesh) (Variables|Default|GetIceMeltA))
  (Variables|Default|SetValueSet true))

(fn OnRep_Phase1_Active?
  (Components|StaticMesh|SetStaticMesh (Variables|Default|GetStaticMesh) (Variables|Default|GetIceMeltB))
  (Variables|Default|SetValueSet true))

(fn OnRep_Phase2_Active?
  (Components|StaticMesh|SetStaticMesh (Variables|Default|GetStaticMesh) (Variables|Default|GetIceMeltC))
  (Variables|Default|SetValueSet true))

(fn OnRep_Phase3_Active?
  (Components|StaticMesh|SetStaticMesh (Variables|Default|GetStaticMesh) (Variables|Default|GetIceMeltD))
  (Variables|Default|SetValueSet true))

(fn OnRep_Phase4_Active?
  (Components|StaticMesh|SetStaticMesh (Variables|Default|GetStaticMesh) (Variables|Default|GetIceMeltE))
  (Variables|Default|SetValueSet true))

(fn OnRep_Phase5_Active?
  (Development|SetHiddeninGame (Variables|Default|GetStaticMesh) true false)
  (Variables|Default|SetValueSet true))

(fn OnRep_Time_D
  (Transformation|SetRelativeTransform (Variables|Default|GetStaticMesh) (Math|Transform|Lerp(Transform) (Variables|Default|GetTransform_D1) (Variables|Default|GetTransform_D2) (Math|Curves|GetFloatValue (Variables|Default|GetCurve) (Variables|Default|GetTime_D))) false false))

(fn OnRep_Time_E
  (Transformation|SetRelativeTransform (Variables|Default|GetStaticMesh) (Math|Transform|Lerp(Transform) (Variables|Default|GetTransform_E1) (Variables|Default|GetTransform_E2) (Math|Curves|GetFloatValue (Variables|Default|GetCurve) (Variables|Default|GetTime_E))) false false))

(fn ConstructionScript
  (bind _staticmesh (Variables|Default|GetStaticMesh))
  (bind _staticmesh_2 (Variables|Default|GetStaticMesh))
  (if (Variables|Default|GetDebug_Active?)
    (Utilities|FlowControl|Sequence
      (:then_0
        (if (== 0 (Variables|Default|GetDebug_Phase))
          (Variables|Default|SetPhase0_Active? true)))
      (:then_1
        (if (== 1 (Variables|Default|GetDebug_Phase))
          (Variables|Default|SetPhase1_Active? true)))
      (:then_2
        (if (== 2 (Variables|Default|GetDebug_Phase))
          (Variables|Default|SetPhase2_Active? true)))
      (:then_3
        (if (== 3 (Variables|Default|GetDebug_Phase))
          (Variables|Default|SetPhase3_Active? true)
          (if (== (Variables|Default|GetTime_D) -1.0)
            (Transformation|SetRelativeTransform _staticmesh (Variables|Default|GetTransform_D1) false false)
            (else
              (Transformation|SetRelativeTransform _staticmesh (Math|Transform|Lerp(Transform) (Variables|Default|GetTransform_D1) (Variables|Default|GetTransform_D2) (Math|Curves|GetFloatValue (Variables|Default|GetCurve) (Variables|Default|GetTime_D))) false false)))))
      (:then_4
        (if (== 4 (Variables|Default|GetDebug_Phase))
          (Variables|Default|SetPhase4_Active? true)
          (if (== (Variables|Default|GetTime_E) -1.0)
            (Transformation|SetRelativeTransform _staticmesh_2 (Variables|Default|GetTransform_E1) false false)
            (else
              (Transformation|SetRelativeTransform _staticmesh_2 (Math|Transform|Lerp(Transform) (Variables|Default|GetTransform_E1) (Variables|Default|GetTransform_E2) (Math|Curves|GetFloatValue (Variables|Default|GetCurve) (Variables|Default|GetTime_E))) false false))))))))
