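; Iceberg prototype ice shard static mesh actor that can remove its mesh, with logic tied to a calendar event.
; On BeginPlay, if the Infinity_On scalable float is 1, it waits for playlist data, checks and watches the named calendar event, and sets SetHidden.
; RepNotifies apply the state: SetHidden destroys the mesh component, while SetVisible unhides it, though nothing here sets SetVisible.
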(event EventBeginPlay
  ; Check if enabled in Data Table
  (if (== (ScalableFloat|GetValueasInteger (Variables|Default|GetInfinity_On)) 1)
    (Class|FortAsyncActionCurrentPlaylistReady|PlaylistDataReadyAsync
      (:OnReady
        (Utilities|FlowControl|Sequence
          (:then_0
            (bind _biseventactive (CalendarEvents|GetCalendarEventInformation (Utilities|String|ToString(Name) (Variables|Default|GetCalendarEventName))))
            (if _biseventactive
              _))
          (:then_1
            (MCPContext|WatchCalendarEvent (Utilities|String|ToString(Name) (Variables|Default|GetCalendarEventName))
              (:OnEventActive
                _))))
        (Variables|Default|SetSetHidden true)))))

(fn OnRep_SetHidden
  (Components|DestroyComponent (Variables|StaticMeshActor|GetStaticMeshComponent)))

(fn OnRep_SetVisible
  (Development|SetHiddeninGame (Variables|StaticMeshActor|GetStaticMeshComponent) false true))

(fn ConstructionScript)
