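; Ice shell actor for the Iceberg event, with one health pool shared across game servers through the mesh network.
; Edge servers forward prop damage; the root subtracts it, persists state to mesh metadata, and edges apply replicated health to the prop.
; An anomaly multiplier, optionally steered by current versus target DPS from the "IceChunk" calendar's remaining time, can only amplify damage.
; Anomaly changes broadcast a cosmetic event, and a hotfix can destroy the shell.
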(event Custom|UpdateHealth
  (params (Damage :type Integer64))
  ;[1 1 1] Root node only
  (bind _returnvalue (CallFunction|HandleAnomalies Damage))
  (bind _output_get (Variables|Default|SetTotalHealth (- (Variables|Default|GetTotalHealth) _returnvalue)))
  ;;[1 1 1] Persist the current health to the mesh service
  (bind _structout (Utilities|Struct|SetmembersinBPIceShellStruct (Variables|Default|GetIceShellMetaData) _output_get))
  (MeshNetworkSubsystem|SetMetaDatawithKey (GameInstanceSubsystems|GetMeshNetworkSubsystem) (Variables|Default|GetMetaDataKey) _structout))

(event Custom|MeshNet_IceShell_SetAnomalyAmount
  (params (Value :type Integer64))
  ;
  (bind _output_get (Variables|Default|SetAnomalyAmount Value))
  (bind _structout (Utilities|Struct|SetmembersinBPIceShellStruct (Variables|Default|GetIceShellMetaData) _output_get))
  (MeshNetworkSubsystem|SetMetaDatawithKey (GameInstanceSubsystems|GetMeshNetworkSubsystem) (Variables|Default|GetMetaDataKey) _structout))

(event Custom|RootSetHealth
  (params (Damage :type Integer64))
  ;[1 1 1] Root node only
  (Variables|Default|SetTotalHealth Damage))

(event Custom|MeshNet_IceShell_OverrideHealth
  (params (InHealth :type Integer64))
  ;
  (bind _output_get (Variables|Default|SetTotalHealth InHealth))
  (bind _structout (Utilities|Struct|SetmembersinBPIceShellStruct (Variables|Default|GetIceShellMetaData) _output_get))
  (MeshNetworkSubsystem|SetMetaDatawithKey (GameInstanceSubsystems|GetMeshNetworkSubsystem) (Variables|Default|GetMetaDataKey) _structout))

(event Custom|MeshMetaDataUpdated
  (bind _iceshellmetadata (Variables|Default|GetIceShellMetaData))
  (bind (_currenthealth_3_80275d8141fa71e07fa6cfb8ffa9581c _anomalyamount_7_87c26a23482ce94a897cfd8d45405301 _anomalyadjustment_11_c1b79b074d8b68edca21a38ebcb22b8c) (Utilities|Struct|BreakBPIceShellStruct _iceshellmetadata))
  ;[1 1 1] Restore initial health value if the root was restarted
  (MeshNetworkSubsystem|GetMetaDatawithKey (GameInstanceSubsystems|GetMeshNetworkSubsystem) (Variables|Default|GetMetaDataKey) _iceshellmetadata)
  (if (>= _currenthealth_3_80275d8141fa71e07fa6cfb8ffa9581c 0)
    (Variables|Default|SetAnomalyAmount _anomalyamount_7_87c26a23482ce94a897cfd8d45405301)
    (Variables|Default|SetEnableAnomalyDynamicAdjustment (Widget|UEFN|ToBool(Integer) _anomalyadjustment_11_c1b79b074d8b68edca21a38ebcb22b8c))
    (CallFunction|RootSetHealth _currenthealth_3_80275d8141fa71e07fa6cfb8ffa9581c)))

(event Custom|MeshNet_IceShell_EnableAnomalyAdjustment
  (params (InValue :type Integer64))
  ;
  (Variables|Default|SetEnableAnomalyDynamicAdjustment (Widget|UEFN|ToBool(Integer) InValue))
  (bind _structout (Utilities|Struct|SetmembersinBPIceShellStruct (Variables|Default|GetIceShellMetaData) InValue))
  (MeshNetworkSubsystem|SetMetaDatawithKey (GameInstanceSubsystems|GetMeshNetworkSubsystem) (Variables|Default|GetMetaDataKey) _structout))

(event Custom|PropDamageRecieved
  (params (Damage :type Integer64))
  ; Game Server
  (Class|BPIceShellDamageComponent|ServerCountDamage (Variables|Default|GetBP_IceChunkMeshNetwork) Damage))

(event EventBeginPlay
  (if (Networking|IsDedicatedServer)
    (Class|FortAsyncActionMeshNetworkReady|MeshNetworkReadyAsync
      (:OnReady
        (Utilities|FlowControl|Sequence
          (:then_0
            (switch Utilities|FlowControl|Switch|SwitchonEMeshNetworkNodeType (MeshNetworking|GetMeshNetworkNodeType (Variables|Default|GetBP_IceChunkMeshNetwork))
              (:Root
                ;[1 1 1] Root node only
                (Default|BindEventtoOnDamageAdded (Variables|Default|GetBP_IceChunkMeshNetwork) (EventDispatchers|CreateEvent :!event Custom|UpdateHealth))
                (MeshNetworkSubsystem|BindEventtoOnMeshMetaDataUpdated (GameInstanceSubsystems|GetMeshNetworkSubsystem) (EventDispatchers|CreateEvent :!event Custom|MeshMetaDataUpdated))
                (CallFunction|MeshMetaDataUpdated)
                ;; Recalculate DPS
                (Variables|Default|SetLastRegisteredHealth (Math|Conversions|ToFloat(Integer64) (Variables|Default|GetTotalHealth)))
                (Utilities|Time|SetTimerbyEvent (delegate Custom|RecalculateDPS) (Math|Conversions|ToFloat(Integer) (Variables|Default|GetDPSRecalculationInterval)) true))
              (:Edge
                ; Game Server
                (Default|BindEventtoOnDamageRecieved (Variables|Default|GetIceShellProp1) (EventDispatchers|CreateEvent :!event Custom|PropDamageRecieved)))))
          (:then_1
            ;[1 1 1] Hotfix to make sure it stays destroyed
            (bind _asfort_game_state_athena (Utilities|Casting|CastToFortGameStateAthena (Game|GetGameState)))
            (if (Variables|/Script/FortniteGame.FortGameStateAthena|GetbPlaylistDataIsLoaded _asfort_game_state_athena)
              (if (== (ScalableFloat|GetValueatLevel (Variables|Default|GetHotfixDestroy)) 1.0)
                (Actor|DestroyActor (Variables|Default|GetIceShellProp1))
                (Actor|DestroyActor))
              (else
                (Playlist|BindEventtoOnCurrentPlaylistLoaded _asfort_game_state_athena (delegate Custom|ContinueSetup))))))))))

(event Custom|RecalculateDPS
  (bind _totalhealth (Variables|Default|GetTotalHealth))
  ;[1 1 1] Root node only
  ;; Recalculate DPS
  (if (Variables|Default|GetEnableAnomalyDynamicAdjustment)
    (CallFunction|GetTargetDPS)
    (Variables|Default|SetCurrentDPS (/ (Math|Conversions|ToFloat(Integer64) (- (Variables|Default|GetLastRegisteredHealth) _totalhealth)) (Variables|Default|GetDPSRecalculationInterval)))
    (Variables|Default|SetLastRegisteredHealth (Math|Conversions|ToFloat(Integer64) _totalhealth))))

(event Custom|ContinueSetup
  (params (PlaylistName :type Name) (PlaylistContextTags :type Gameplay%20Tag%20Container%20Structure%20%28by%20ref%29))
  ;[1 1 1] Hotfix to make sure it stays destroyed
  (if (== (ScalableFloat|GetValueatLevel (Variables|Default|GetHotfixDestroy)) 1.0)
    (Actor|DestroyActor (Variables|Default|GetIceShellProp1))
    (Actor|DestroyActor)))

(fn GetTargetDPS
  (bind (_biseventactive _timesincebegin _timeuntilend) (CalendarEvents|GetCalendarEventInformation "IceChunk"))
  (if _biseventactive
    (Variables|Default|SetTargetDPS (Math|Float|SafeDivide (Math|Conversions|ToFloat(Integer64) (Variables|Default|GetTotalHealth)) (Math|Timespan|GetTotalSeconds _timeuntilend)))
    (else
      (Variables|Default|SetTargetDPS 0.0))))

(fn HandleAnomalies
  (params (InValue :type Integer64))
  (return (Math|Integer64|Max(Integer64) (Math|Conversions|ToInteger64(Integer) (Math|Float|Truncate (Math|Float|SelectFloat (Math|Float|SelectFloat (Variables|Default|GetTargetDPS) (Utilities|Operators|Multiply (/ (Math|Conversions|ToFloat(Integer64) (Variables|Default|GetAnomalyAmount)) 1e+02) (Math|Float|SafeDivide (Variables|Default|GetTargetDPS) (Variables|Default|GetCurrentDPS)) InValue) (Math|Float|NearlyEqual(Float) (Variables|Default|GetCurrentDPS))) (* (/ (Math|Conversions|ToFloat(Integer64) (Variables|Default|GetAnomalyAmount)) 1e+02) (Math|Conversions|ToFloat(Integer64) InValue)) (Variables|Default|GetEnableAnomalyDynamicAdjustment)))) (Variables|GetInValue))))

(fn OnRep_AnomalyAmount
  (Networking|CanExecuteCosmeticEvents
    (:True
      (GameplayEventRouter|BroadcastStatefulEvent (GameplayEventRouter|GetGlobalEventRouter) (Utilities|Struct|MakeFigmentEventIceShellAnomaly (Math|Float|MapRangeClamped (Math|Conversions|ToFloat(Integer64) (Variables|Default|GetAnomalyAmount)) 5e+01 2.5e+02 0.0 1.0)) (Fortnite|GetFirstLocalFortPlayerController true)))))

(fn OnRep_TotalHealth
  (switch Utilities|FlowControl|Switch|SwitchonEMeshNetworkNodeType (MeshNetworking|GetMeshNetworkNodeType (Variables|Default|GetBP_IceChunkMeshNetwork))
    (:Edge
      (Class|BPIceShellProp|EdgeServerApplyDamage (Variables|Default|GetIceShellProp1) (Variables|Default|GetTotalHealth)))))

(fn ConstructionScript)
