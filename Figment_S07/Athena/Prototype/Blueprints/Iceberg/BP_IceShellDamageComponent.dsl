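; BP_IceShellDamageComponent is a component that reports damage counted against an ice shell in an Iceberg prototype.
; Its ServerCountDamage event takes an integer damage amount and broadcasts it through the DamageAdded dispatcher.
; BeginPlay and Tick are empty, so the component keeps no running total in this graph and leaves the response to listeners.
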(event Custom|ServerCountDamage
  (params (DamageApplied :type Integer64))
  (Default|CallOnDamageAdded DamageApplied))

(event EventBeginPlay)

(event EventTick
  (params (DeltaSeconds :type Float%20%28single-precision%29)))
