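; Iceberg ice shell prop with 64-bit "big health" and a custom health bar, destroyed when its health total reaches zero.
; Server damage broadcasts OnDamageRecieved, while EdgeServerApplyDamage applies incoming totals, capped at current health unless root-server override is on.
; A randomized RepNotify value triggers client bar updates within range, and sphere overlap keeps the bar visible.
; Setup waits for playlist load, then sets max health via GameplayEffect or follows a no-damage hotfix.
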(event Custom|EdgeServerApplyDamage
  (params (TotalHealth :type Integer64))
  (bind _enablerootserverhealthoverride (Variables|Default|GetEnableRootServerHealthOverride))
  (bind _returnvalue (ScalableFloat|GetValueasBool _enablerootserverhealthoverride))
  (bind _bighealth (Variables|BuildingPropBigHealth|GetBigHealth))
  (bind _returnvalue_2 (Math|Integer64|Min(Integer64) _bighealth TotalHealth))
  (bind _returnvalue_3 (Math|Conversions|ToInteger(Integer64) _returnvalue_2))
  (bind _returnvalue_4 (Math|Conversions|ToInteger(Integer64) TotalHealth))
  (bind _returnvalue_5 (Math|Integer|SelectInt _returnvalue_4 _returnvalue_3 _returnvalue))
  (bind _returnvalue_6 (Math|Conversions|ToInteger64(Integer) _returnvalue_5))
  ;[1 1 1] Leaf server only
  (if (> TotalHealth 0)
    (CallFunction|SetBigHealth _returnvalue_6 true)
    (Health|SetHealth (Math|Float|Min(Float) (Health|GetMaxHealth) (Math|Conversions|ToFloat(Integer64) _returnvalue_6)))
    (CallFunction|ReplicateHealthBarValuestoClients)
    (Networking|ForceNetUpdate)
    (else
      (Health|SetHealth 0.0)
      ;;[1 1 1] DestroyActor on BuildingActor calls Die()
      (Actor|DestroyActor))))

(event Custom|ReplicateHealthBarValuesToClients
  (Variables|Default|SetHealthPercentage (CallFunction|GetHealthValuePercent))
  (Variables|Default|SetUpdateHealthForClient (Math|Random|RandomFloat)))

(event EventOnDamageServer
  (params (Damage :type Float%20%28single-precision%29) (DamageTags :type Gameplay%20Tag%20Container%20Structure%20%28by%20ref%29) (Momentum :type Vector :default "0, 0, 0") (HitInfo :type Hit%20Result%20Structure%20%28by%20ref%29) (InstigatedBy :type Controller%20Object%20Reference) (DamageCauser :type Actor%20Object%20Reference) (EffectContext :type Gameplay%20Effect%20Context%20Handle%20Structure))
  (|CallParentFunction Damage DamageTags Momentum HitInfo InstigatedBy DamageCauser EffectContext)
  (if (not (Utilities|Equal(Object) DamageCauser self))
    (Default|CallOnDamageRecieved (Math|Conversions|ToInteger64(Integer) (Math|Float|Truncate Damage)))
    (CallFunction|ReplicateHealthBarValuestoClients)))

(event EventBeginPlay
  (Collision|SetCollisionResponsetoChannel (Variables|Debug|GetStaticMeshComponent) "ECC_GameTraceChannel13" "ECR_Block")
  (bind _asfort_game_state_athena (Utilities|Casting|CastToFortGameStateAthena (Game|GetGameState)))
  (if (Variables|/Script/FortniteGame.FortGameStateAthena|GetbPlaylistDataIsLoaded _asfort_game_state_athena)
    (if (ScalableFloat|GetValueasBool (Variables|Default|GetHotfixNoDamage))
      (Fortnite|SetCanbeDamaged self)
      (Variables|BuildingPropBigHealth|SetbShowHealthBar false)
      (Components|DestroyComponent (Variables|Default|GetFortWidget))
      (else
        (Networking|SwitchHasAuthority
          (:Authority
            (GameplayEffects|ApplyGameplayEffectToSelf :self (Ability|GetAbilitySystemComponent self) :GameplayEffectClass "/Figment_S07/Athena/Prototype/Blueprints/Iceberg/GE_MaxHealthSet.GE_MaxHealthSet_C" :Level 0.0))
          (:Remote
            (EventDispatchers|BindEventtoOnDied (delegate Custom|Effect))
            (bind _fortwidget (Variables|Default|GetFortWidget)
              (:then
                (bind _asfigment_big_health_bar (Utilities|Casting|CastToFigmentBigHealthBar (UserInterface|GetUserWidgetObject _fortwidget)))
                (Variables|Default|SetHeathWidget _asfigment_big_health_bar)
                (CallFunction|StartListeningforHealthWidgetUpdates))
              (:else))))))
    (else
      (Playlist|BindEventtoOnCurrentPlaylistLoaded _asfort_game_state_athena (delegate Custom|ContinueSetup)))))

(event Custom|StartListeningForHealthWidgetUpdates
  (bind _heathwidget (Variables|Default|GetHeathWidget))
  ;[1 1 1] Tell Clients to start listening when the UI is supposed to be Visible or Not
  (Utilities|FlowControl|Sequence
    (:then_0
      (bind _fortwidget (Variables|Default|GetFortWidget)
        (:then
          (Components|Tick|SetComponentTickEnabled _fortwidget (Widget|IsVisible _heathwidget)))
        (:else)))
    (:then_1
      (Default|BindEventtoVisibilityChanged _heathwidget (delegate Custom|WidgetVisibilityChanged)))))

(event Custom|ContinueSetup
  (params (PlaylistName :type Name) (PlaylistContextTags :type Gameplay%20Tag%20Container%20Structure%20%28by%20ref%29))
  (if (ScalableFloat|GetValueasBool (Variables|Default|GetHotfixNoDamage))
    (Fortnite|SetCanbeDamaged self)
    (Variables|BuildingPropBigHealth|SetbShowHealthBar false)
    (Components|DestroyComponent (Variables|Default|GetFortWidget))
    (else
      (Networking|SwitchHasAuthority
        (:Authority
          (GameplayEffects|ApplyGameplayEffectToSelf :self (Ability|GetAbilitySystemComponent self) :GameplayEffectClass "/Figment_S07/Athena/Prototype/Blueprints/Iceberg/GE_MaxHealthSet.GE_MaxHealthSet_C" :Level 0.0))
        (:Remote
          (EventDispatchers|BindEventtoOnDied (delegate Custom|Effect))
          (bind _fortwidget (Variables|Default|GetFortWidget)
            (:then
              (bind _asfigment_big_health_bar (Utilities|Casting|CastToFigmentBigHealthBar (UserInterface|GetUserWidgetObject _fortwidget)))
              (Variables|Default|SetHeathWidget _asfigment_big_health_bar)
              (CallFunction|StartListeningforHealthWidgetUpdates))
            (:else)))))))

(event Custom|Effect
  (params (DamagedActor :type Actor%20Object%20Reference) (Damage :type Float%20%28single-precision%29) (InstigatedBy :type Controller%20Object%20Reference) (DamageCauser :type Actor%20Object%20Reference) (HitLocation :type Vector) (FHitComponent :type Primitive%20Component%20Object%20Reference) (BoneName :type Name) (Momentum :type Vector)))

(event Custom|WidgetVisibilityChanged
  (params (bVisible :type Boolean))
  ;[1 1 1] Tell Clients to start listening when the UI is supposed to be Visible or Not
  (bind _fortwidget (Variables|Default|GetFortWidget)
    (:then
      (Components|Tick|SetComponentTickEnabled _fortwidget bVisible))
    (:else)))

(event Custom|StopListeningForHealthWidgetUpdates
  ;[1 1 1] Tell Clients to start listening when the UI is supposed to be Visible or Not
  (bind _heathwidget (Variables|Default|GetHeathWidget)
    (:then
      (Default|UnbindEventfromVisibilityChanged _heathwidget (delegate Custom|WidgetVisibilityChanged)))
    (:else)))

(event Collision|EventActorBeginOverlap
  (params (OtherActor :type Actor%20Object%20Reference)))

(event Custom|UpdateCustomHealthBar
  ;[1 1 1] Update Health Bar to Relevant Players
  (bind _array_element (Utilities|Array|ForEachLoopwithBreak (Fortnite|GetLocalFortPlayerControllers true))
    (:LoopBody
      (if (< (Class|AbilityFunctions|GetDistanceBetweenActors (Pawn|GetControlledPawn _array_element) self) (ScalableFloat|GetValueatLevel (Variables|Default|GetMaxDistanceFromDamageInstigatorToDrawHealthBar)))
        (bind _heathwidget (Variables|Default|GetHeathWidget)
          (:then
            (Class|FigmentBigHealthBar|EventUpdateStatus _heathwidget (Health|GetHealthPercent) (Utilities|Text|ToText(Integer64) (Variables|BuildingPropBigHealth|GetBigHealth)))
            ; Workaround to fix a bug that initializes the health bar with the default health amount when it is called for the first time.
            (if (Variables|Default|GetRecallUpdateVisualsOnce)
              _)
            (Variables|Default|SetRecallUpdateVisualsOnce true)
            (Utilities|FlowControl|DelayUntilNextTick)
            (CallFunction|UpdateCustomHealthBar))
          (:else))))))

(event EventTick
  (params (DeltaSeconds :type Float%20%28single-precision%29)))

(event OnComponentBeginOverlap(Sphere)
  (params (OverlappedComponent :type Primitive%20Component%20Object%20Reference) (OtherActor :type Actor%20Object%20Reference) (OtherComp :type Primitive%20Component%20Object%20Reference) (OtherBodyIndex :type Integer) (bFromSweep :type Boolean) (SweepResult :type Hit%20Result%20Structure%20%28by%20ref%29))
  (bind _asplayer_pawn_athena (Utilities|Casting|CastToPlayerPawn_Athena OtherActor))
  (Utilities|Casting|CastToPlayerController (Pawn|GetController _asplayer_pawn_athena))
  (Networking|SwitchHasAuthority
    (:Remote
      (CallFunction|UpdateCustomHealthBar)
      (bind _heathwidget (Variables|Default|GetHeathWidget)
        (:then
          (Class|FigmentBigHealthBar|SetKeepVisibilityOn _heathwidget true :!class /Figment_EventsUI/UI/InGame/HUD/FigmentBigHealthBar.FigmentBigHealthBar_C))
        (:else)))))

(event OnComponentEndOverlap(Sphere)
  (params (OverlappedComponent :type Primitive%20Component%20Object%20Reference) (OtherActor :type Actor%20Object%20Reference) (OtherComp :type Primitive%20Component%20Object%20Reference) (OtherBodyIndex :type Integer))
  (bind _asplayer_pawn_athena (Utilities|Casting|CastToPlayerPawn_Athena OtherActor))
  (Utilities|Casting|CastToPlayerController (Pawn|GetController _asplayer_pawn_athena))
  (Networking|SwitchHasAuthority
    (:Remote
      (bind _heathwidget (Variables|Default|GetHeathWidget)
        (:then
          (Class|FigmentBigHealthBar|SetKeepVisibilityOn _heathwidget false :!class /Figment_EventsUI/UI/InGame/HUD/FigmentBigHealthBar.FigmentBigHealthBar_C))
        (:else)))))

(fn OnRep_HealthPercentage)

(fn OnRep_UpdateHealthForClient
  (Networking|SwitchHasAuthority
    (:Remote
      (CallFunction|UpdateCustomHealthBar))))

(fn ShouldDie
  (params (Damage :type Float%20%28single-precision%29) (EventInstigator :type Controller%20Object%20Reference) (DamageCauser :type Actor%20Object%20Reference) (bIsTest :type Boolean) (EffectContext :type Gameplay%20Effect%20Context%20Handle%20Structure%20%28by%20ref%29))
  (return true))

(fn ConstructionScript
  (Variables|BuildingPropBigHealth|SetbShowHealthBar false))
