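; Iceberg static mesh actor that swaps to a second mesh when calendar event "C1" (noted as the Beacon Event) becomes active.
; On BeginPlay, unless DebugShowPhase1 is set, the authority watches the event and sets Phase1Active.
; OnRep_Phase1Active then picks between the two meshes, so the change replicates.
; The construction script selects the mesh from DebugShowPhase1, allowing a preview of the phase.
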(event EventBeginPlay
  (if (not (Variables|Default|GetDebugShowPhase1))
    (Networking|SwitchHasAuthority
      (:Authority
        (MCPContext|WatchCalendarEvent "C1"
          (:OnEventActive
            (Variables|Default|SetPhase1Active true))) ; Beacon Event
      ))))

(event Collision|EventActorBeginOverlap
  (params (OtherActor :type Actor%20Object%20Reference)))

(event EventTick
  (params (DeltaSeconds :type Float%20%28single-precision%29)))

(fn OnRep_Phase1Active
  (Components|StaticMesh|SetStaticMesh (Variables|StaticMeshActor|GetStaticMeshComponent) (select (Variables|Default|GetPhase1Active) (Variables|Default|GetStaticMesh2) (Variables|Default|GetStaticMesh1))))

(fn ConstructionScript
  (Components|StaticMesh|SetStaticMesh (Variables|StaticMeshActor|GetStaticMeshComponent) (select (Variables|Default|GetDebugShowPhase1) (Variables|Default|GetStaticMesh2) (Variables|Default|GetStaticMesh1))))
