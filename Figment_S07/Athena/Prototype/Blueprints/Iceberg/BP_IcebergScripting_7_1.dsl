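; Season 7 prototype iceberg scripting Blueprint that removes its actor, tied to calendar events C3 and C3A.
; On BeginPlay with authority, it watches those events unless DebugDelete bypasses the check, and leads into setting DeleteMesh.
; The DeleteMesh RepNotify then destroys the actor, so removal follows the replicated flag.
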(event EventBeginPlay
  (Networking|SwitchHasAuthority
    (:Authority
      (if (not (Variables|Default|GetDebugDelete))
        (MCPContext|WatchCalendarEvents (Utilities|Array|MakeArray "C3" "C3A")
          (:OnEventActive
            _)))
      (Variables|Default|SetDeleteMesh true))))

(fn OnRep_DeleteMesh
  (if (Variables|Default|GetDeleteMesh)
    (Actor|DestroyActor)))

(fn ConstructionScript)
