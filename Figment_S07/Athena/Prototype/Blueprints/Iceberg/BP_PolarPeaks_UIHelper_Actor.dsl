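; BP_PolarPeaks_UIHelper_Actor is a helper actor that stamps the Polar Peaks iceberg onto the map through the game state's dynamic land system.
; SetupPolarPeaksIceStamp stores the target position, clears the previous stamp, selects the full or half ice stamp by version, then calls SetDynamicLand with position, rotation offset and shadow-layer setting.
; The actor's object name keys the stamp, so ClearMapIcons removes it through RemoveDynamicLand.
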(event Custom|SetupPolarPeaksIceStamp
  (params (TargetPosition :type Vector) (IceStampVersion :type Integer))
  (Variables|Default|SetTargetPosition TargetPosition)
  (CallFunction|ClearMapIcons)
  (switch int IceStampVersion
    (:0
      (Variables|Default|SetCurrentIslceStamp (Variables|Default|GetIceStampFull)))
    (:1
      (Variables|Default|SetCurrentIslceStamp (Variables|Default|GetIceStampHalf))))
  (Class|FortGameStateAthena|SetDynamicLand (Fortnite|Athena|GameState|GetGameStateAthena) (Utilities|String|StringToName (Utilities|GetObjectName self)) (Variables|Default|GetTargetPosition) (Variables|Default|GetRotationOffset) (Variables|Default|GetCurrentIslceStamp) (Variables|Default|GetUseShadowLayer)))

(fn ClearMapIcons
  (Class|FortGameStateAthena|RemoveDynamicLand (Fortnite|Athena|GameState|GetGameStateAthena) (Utilities|String|StringToName (Utilities|GetObjectName self)) (Variables|Default|GetUseShadowLayer)))

(fn ConstructionScript)
